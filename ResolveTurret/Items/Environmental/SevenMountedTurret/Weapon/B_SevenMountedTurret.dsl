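; Blueprint for the Seven mounted turret weapon, handling its muzzle flash effects on the host vehicle.
; Host setup caches the vehicle and flags an Armored Battle Bus by gameplay tag, with a leftover debug print.
; Muzzle flashes attach to the vehicle mesh's left (off-hand) or right socket, using "_b" socket names on the bus.
; Construction sets a red turret flash particle; shell and fire FX events add no effects of their own.
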(event EventMuzzleFlashFX
  (params (Persistent%20Fire :type Boolean) (OffHand :type Boolean))
  ;[1 1 1] Play Muzzle Flash FX
  (Utilities|IsValid (Variables|Default|GetAthenaVehicle)
    (:Is%20Valid
      (if OffHand
        (bind _returnvalue (Effects|SpawnFXSystemAttached (Variables|Default|GetMuzzleFlashFX) (Class|FortVehicleInterface|GetSkeletalMeshComponent (Variables|Default|GetAthenaVehicle)) (select (Variables|Default|GetIsArmoredBattleBus) "muzzleLeft_b" "muzzleLeft"))) ; Left
        (Effects|Components|ParticleSystem|SetFloatParameter _returnvalue "User.MuzzleLeft" 1.0)
        (else
          (Effects|SpawnFXSystemAttached (Variables|Default|GetMuzzleFlashFX) (Class|FortVehicleInterface|GetSkeletalMeshComponent (Variables|Default|GetAthenaVehicle)) (select (Variables|Default|GetIsArmoredBattleBus) "muzzleRight_b" "muzzleRight")) ; Right
        )))))

(event EventOnPlayWeaponFireFX
  (params (bPersistentFire :type Boolean) (bSecondaryFire :type Boolean)))

(event EventActivateShellsFX
  (params (Bool :type Boolean) (OffHand :type Boolean))
  ;[1 1 1] Play Muzzle Flash FX
  (|CallParentFunction Bool OffHand)
  (Utilities|IsValid (Variables|Default|GetAthenaVehicle)
    (:Is%20Valid
      (if OffHand
        _))))

(event Setup|EventOnHostVehicleSetup
  ;[1 1 1] Check if Host vehicle is not battle bus
  (|CallParentFunction)
  (bind _asfort_athena_vehicle (Utilities|Casting|CastToFortAthenaVehicle (Utilities|ToObject(Interface) (Vehicle|GetHostVehicle))))
  (Variables|Default|SetAthenaVehicle _asfort_athena_vehicle)
  (if (GameplayTags|HasMatchingGameplayTag _asfort_athena_vehicle "(TagName=\"Athena.Vehicle.ArmoredBattleBus\")")
    (Development|PrintString "IM THE BATTLE BUS" true true "(R=0.000000,G=0.660000,B=1.000000,A=1.000000)" 15.0)
    (Variables|Default|SetIsArmoredBattleBus true)))

(fn ConstructionScript
  (Variables|Effects|SetMuzzleParticleSystem "/Game/Gadgets/Assets/Turret_Gun/Particles/P_TurretGun_MFlash_Red.P_TurretGun_MFlash_Red")
  (|CallParentFunction))
